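# Conversation database for the SquibblySquid Deep 02 NPC, supplying its food-critic flavored greeting and access to shared NPC conversation options.
# The tagged entry point leads to one speech node that picks contextual lines by first-ever meeting, per-match conversation count, or player outfit (SquibblySquid, TeriyakiFish).
# First-meeting and outfit lines take priority from named curve rows.
# The speech then branches through link nodes to the NPCLibrary Quests, BasicServices and RandomServices entries.
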ConversationDatabase /NPCLibrary/NPCs/SquibblySquid_Deep/SquibblySquid_Deep_02/NPCLibrary_Conversation_SquibblySquid_Deep_02
entry_points [{tag=Conversation.Entry.SquibblySquidDeep.02, destinations [#1]}]
nodes [
 ConversationEntryPointNode_7 {
  id=#1
  class=ConversationEntryPointNode
  outputs [#2]
  properties {entryTag {tagName=Conversation.Entry.SquibblySquidDeep.02}}
 }
 Speech_C_4 {
  id=#2
  class=Speech_C
  outputs [#3, #4, #5]
  properties {
   nodeName=""
   generalConfig {
    defaultMessage="So many opinions, and so many ways to say them!"
    contextualMessages [
     {
      message="If you're looking for feedback, I have THOUGHTS."
      contextRequirements [{requirement {class=FortControllerRequirement_IsFirstEverConversationWithNPC}}]
      priority {curve {rowName=Tandem.Speech.Priority.FirstEverInteraction}}
     }
     {
      message=Hmm... inviting bite... 'turpid' aftertaste? No...
      contextRequirements [{requirement {
         class=FortControllerRequirement_CheckNPCMatchConversationCount
         properties {nPCUniqueIdTag {tagName=AISpawnerData.Type.Tandem.SquibblySquidDeep.02}}
        }}]
      priority {curve {curveTable=null, rowName=None}}
     }
     {
      message=How can I balance your flavor profile today?
      contextRequirements [{requirement {
         class=FortControllerRequirement_CheckNPCMatchConversationCount
         properties {nPCUniqueIdTag {tagName=AISpawnerData.Type.Tandem.SquibblySquidDeep.02}}
        }}]
      priority {curve {curveTable=null, rowName=None}}
     }
     {
      message="So many opinions, and so many ways to say them!"
      contextRequirements [{requirement {
         class=FortControllerRequirement_CheckNPCMatchConversationCount
         properties {nPCUniqueIdTag {tagName=AISpawnerData.Type.Tandem.SquibblySquidDeep.02}}
        }}]
      priority {curve {curveTable=null, rowName=None}}
     }
     {
      message="Warm plates and full hearts!"
      contextRequirements [{requirement {class=FortControllerRequirement_HasCID, properties {allowedCIds [/BRCosmetics/Athena/Items/Cosmetics/Characters/Character_SquibblySquid]}}}]
      priority {curve {rowName=Tandem.Speech.Priority.OutfitReaction}}
     }
     {
      message="I thought I sensed a foul taste in my mouth. "
      contextRequirements [{requirement {
         class=FortControllerRequirement_HasCID
         properties {allowedCIds [&CID_315_Athena_Commando_M_TeriyakiFish]}
        }}]
      priority {curve {rowName=Tandem.Speech.Priority.OutfitReaction}}
     }
    ]
   }
  }
 }
 ConversationLinkNode_2 {id=#5, class=ConversationLinkNode, properties {remoteEntryTag {tagName=Conversation.Entry.RandomServices}}}
 ConversationLinkNode_0 {id=#4, class=ConversationLinkNode, properties {remoteEntryTag {tagName=Conversation.Entry.BasicServices}}}
 ConversationLinkNode_1 {id=#3, class=ConversationLinkNode, properties {remoteEntryTag {tagName=Conversation.Entry.NPCLibrary.Quests}}}
]
